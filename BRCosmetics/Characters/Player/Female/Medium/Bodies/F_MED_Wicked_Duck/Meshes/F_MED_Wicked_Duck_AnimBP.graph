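# Animation Blueprint for the female Wicked Duck body cosmetic, copying the attached parent's pose and adding corrective bone adjustments.
# Two knee-constraint setups blend on crouch state, then feed shoulder-pad rotation driven by right upper-arm twist.
# A weighted multi-bone constraint places the tail-fix master bone, with an extra tail offset while crouching.
# Each update also checks backbling and, for long (cape) pieces, sets the body material's Hide Element parameters.
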AnimBlueprint /BRCosmetics/Characters/Player/Female/Medium/Bodies/F_MED_Wicked_Duck/Meshes/F_MED_Wicked_Duck_AnimBP
target_skeleton=&/F_MED_Wicked_Duck_Skeleton
graphs [
 AnimGraph {nodes [
   n1 = "Constraint - dyn_l_knee (PP)"!(
    ComponentPose=Local To Component(LocalPose=Copy Pose From Mesh() {bUseAttachedParent=true})
    Alpha=1
    ConstraintWeights_0=0.5
    ConstraintWeights_1:float=0.5
   ) -> (Pose) {
    alphaBoolBlend {blendOption=Linear}
    boneToModify {boneName=dyn_l_knee}
    constraintSetup [{targetBone {boneName=thigh_twist_01_l}}, {targetBone {boneName=calf_l}}]
    constraintWeights [1,1]
   }
   n2 = "Constraint - dyn_r_knee (PP)"!(ComponentPose=@n1, Alpha=1, ConstraintWeights_0=0.5, ConstraintWeights_1:float=0.5) -> (Pose) {
    alphaBoolBlend {blendOption=Linear}
    boneToModify {boneName=dyn_r_knee}
    constraintSetup [{targetBone {boneName=thigh_twist_01_r}}, {targetBone {boneName=calf_r}}]
    constraintWeights [1,1]
   }
   n3 = Component To Local(ComponentPose=@n2)
   n4 = "Constraint - dyn_l_knee (PP)"(
    ComponentPose=Local To Component(LocalPose=Copy Pose From Mesh() {bUseAttachedParent=true})
    Alpha=1
    ConstraintWeights_0=0
    ConstraintWeights_1=1
   ) {
    alphaBoolBlend {blendOption=Linear}
    boneToModify {boneName=dyn_l_knee}
    constraintSetup [{targetBone {boneName=thigh_twist_01_l}}, {targetBone {boneName=calf_l}}]
    constraintWeights [1,1]
   }
   n5 = "Constraint - dyn_r_knee (PP)"(ComponentPose=@n4, Alpha=1, ConstraintWeights_0=0.3, ConstraintWeights_1=0.7) {
    alphaBoolBlend {blendOption=Linear}
    boneToModify {boneName=dyn_r_knee}
    constraintSetup [{targetBone {boneName=thigh_twist_01_r}}, {targetBone {boneName=calf_r}}]
    constraintWeights [1,1]
   }
   n6 = "Transform (Modify) Bone - Bone: dyn_r_knee"(ComponentPose=@n5, Alpha=1, Translation="0,0.5,0", Rotation="0,0,0", Scale="1,1,1") {
    rotationMode=BMM_Additive
    rotationSpace=BCS_BoneSpace
    scaleMode=BMM_Additive
    scaleSpace=BCS_BoneSpace
    translationMode=BMM_Additive
    translationSpace=BCS_BoneSpace
    alphaBoolBlend {blendOption=Linear}
    boneToModify {boneName=dyn_r_knee}
    translation {y=0.5}
   }
   n7 = Component To Local(ComponentPose=@n6)
   n8 = Get bIsCrouching()
   n9 = "Transform (Modify) Bone - Bone: dyn_r_shoulderPad"(
    ComponentPose=Local To Component(LocalPose=Blend Poses by bool(BlendPose_0=@n7, BlendPose_1=@n3, BlendTime_0=0.2, BlendTime_1=0.2, bActiveValue=@n8) {blendType=Linear})
    Alpha=Get clav_up_r() -> (clav_up_r:double)
    Translation="0,0,0"
    Rotation="20,0,0"
    Scale="1,1,1"
   ) {
    rotationMode=BMM_Additive
    translationMode=BMM_Additive
    alphaBoolBlend {blendOption=Linear}
    alphaScaleBiasClamp {outRange {max=0}}
    boneToModify {boneName=dyn_r_shoulderPad}
    rotation {pitch=20}
   }
   n10 = "Constraint - dyn_tailFix_MSTR (PPPPPPPP)"!(
    ComponentPose=@n9
    Alpha=1
    ConstraintWeights_0=0.0036
    ConstraintWeights_1:float=0.0857
    ConstraintWeights_2:float=0
    ConstraintWeights_3:float=0
    ConstraintWeights_4:float=0.873
    ConstraintWeights_5:float=0
    ConstraintWeights_6:float=0.0004
    ConstraintWeights_7:float=0.8196
   ) -> (Pose) {
    alphaBoolBlend {blendOption=Linear}
    boneToModify {boneName=dyn_tailFix_MSTR}
    constraintSetup [
     {targetBone {boneName=thigh_r}}
     {targetBone {boneName=thigh_twist_01_r}}
     {targetBone {boneName=spine_01}}
     {targetBone {boneName=spine_02}}
     {targetBone {boneName=thigh_twist_01_l}}
     {targetBone {boneName=spine_04}}
     {targetBone {boneName=thigh_l}}
     {targetBone {boneName=pelvis}}
    ]
    constraintWeights [1,1,1,1,1,1,1,1]
   }
   n11 = "Transform (Modify) Bone - Bone: dyn_tailFix"(ComponentPose=@n10, Alpha=SelectFloat(A=1, B=0, bPickA=Get bIsCrouching()), Translation="0,-4,-2", Rotation="0,0,0", Scale="1,1,1") {
    rotationMode=BMM_Additive
    scaleMode=BMM_Additive
    translationMode=BMM_Additive
    alphaBoolBlend {blendOption=Linear}
    boneToModify {boneName=dyn_tailFix}
    translation {y=-4, z=-2}
   }
   Output Pose(Result=Component To Local(ComponentPose=@n11)) {layerGroup=None}
  ]}
 EventGraph {nodes [
   Event BlueprintUpdateAnimation()
   > n1 = Sequence()
   > Set clav_up_r(clav_up_r:double=GetAlphaFromReferenceTransformDelta(
     SkeletalMeshComponent=Get Body_Mesh()
     BoneName=upperarm_r
     BaseName=clavicle_r
     DeltaComponentType=QuaternionTwist
     InputRange=(X=0,Y=-120)
     TwistAxis="0,0,1"
    )) -> (Output_Get:double)
   n2 = CheckForBackbling(LongBackblingThreshold=40)
   Set isWearingBackbling(execute=@n1.then_1, isWearingBackbling=@n2)
   > Set isWearingCape(isWearingCape=@n2.bIsBackblingLong)
   n3 = Get isWearingBackbling()
   n4 = Get isWearingCape() -> (isWearingCape:bool)
   > Branch(Condition=OR Boolean(A=Get isWearingCape(), B=AND Boolean(A=@n3, B=@n4)))
   > n5 = Get Dynamic Material Instance(SkeletalMeshComponent=GetPartSkeletalMeshComponent())
   > SetScalarParameterValue(self=@n5, ParameterName=Hide Element 02, Value=1)
   > SetScalarParameterValue(self=@n5, ParameterName=Hide Element 01, Value=0)
   Event BlueprintInitializeAnimation()
   > Set Body_Mesh(Body_Mesh=GetPartSkeletalMeshComponent())
   Event OnCharacterCustomizationLoaded()
  ]}
]
variables [
 arm_up {type=number}
 armForward_R {type=number}
 armUp_R {type=number}
 Body_Mesh {type=SkeletalMeshComponent}
 clav_up_l {type=number}
 clav_up_plus {type=number}
 clav_up_r {type=number}
 clav_up_r_plus {type=number}
 hoodFix_alpha_r {type=number}
 isWearingBackbling {type=boolean}
 isWearingCape {type=boolean}
]
