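# Physics asset for the eyes mesh part of the female medium Wicked Duck player character, with capsule bodies for head, neck, four eyelids, a headbob and two wires.
# The neck bodies chain to the head, with the headbob and first wire joined to the head and the second wire to neck_02, all with free rotation.
# Each eyelid joins the head through a hinge-like constraint: wide on one swing axis, nearly locked on the others.
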PhysicsAsset /BRCosmetics/Characters/Player/Female/Medium/Bodies/F_MED_Wicked_Duck/Meshes/Parts/F_MED_Wicked_Duck_Eyes_Physics
bodies {
 dyn_headbob_MSTR [{shapeType=Capsule, length=0.505, radius=0.505, rotation=<rot 0 -90 0>}]
 dyn_l_btmLid [{shapeType=Capsule, length=1.674, radius=1.4141, center=<-2.1998 0.0001 -0.2586>, rotation=<rot -90 0 0>}]
 dyn_l_topLid [{shapeType=Capsule, length=1.674, radius=1.4141, center=<-2.1998 0.0001 -0.2586>, rotation=<rot -90 0 0>}]
 dyn_r_btmLid [{shapeType=Capsule, length=1.674, radius=1.4141, center=<2.067 -0.0001 0>, rotation=<rot -90 0 0>}]
 dyn_r_topLid [{shapeType=Capsule, length=1.674, radius=1.4141, center=<2.067 -0.0001 0>, rotation=<rot -90 0 0>}]
 dyn_wire_01_MSTR [{shapeType=Capsule, length=0.25, radius=0.25}]
 dyn_wire_02_MSTR [{shapeType=Capsule, length=0.25, radius=0.25}]
 head [{shapeType=Capsule, length=2.4316, radius=11.767, center=<7.4641 0 0.0001>, rotation=<rot -90 0 0>}]
 neck_01 [{shapeType=Capsule, length=2, radius=2}]
 neck_02 [{shapeType=Capsule, length=2, radius=2}]
}
constraints [
 {
  bone1Name=dyn_headbob_MSTR
  bone2Name=head
  swing1LimitDegrees=45
  swing1Motion=Free
  swing2LimitDegrees=45
  swing2Motion=Free
  twistLimitDegrees=45
  twistMotion=Free
 }
 {
  bone1Name=head
  bone2Name=neck_02
  swing1LimitDegrees=45
  swing1Motion=Free
  swing2LimitDegrees=45
  swing2Motion=Free
  twistLimitDegrees=45
  twistMotion=Free
 }
 {
  bone1Name=neck_02
  bone2Name=neck_01
  swing1LimitDegrees=45
  swing1Motion=Free
  swing2LimitDegrees=45
  swing2Motion=Free
  twistLimitDegrees=45
  twistMotion=Free
 }
 {
  bone1Name=dyn_wire_01_MSTR
  bone2Name=head
  swing1LimitDegrees=45
  swing1Motion=Free
  swing2LimitDegrees=45
  swing2Motion=Free
  twistLimitDegrees=45
  twistMotion=Free
 }
 {
  bone1Name=dyn_wire_02_MSTR
  bone2Name=neck_02
  swing1LimitDegrees=45
  swing1Motion=Free
  swing2LimitDegrees=45
  swing2Motion=Free
  twistLimitDegrees=45
  twistMotion=Free
 }
 {bone1Name=dyn_l_btmLid, bone2Name=head, swing1LimitDegrees=60, swing2LimitDegrees=1, twistLimitDegrees=1}
 {bone1Name=dyn_l_topLid, bone2Name=head, swing1LimitDegrees=60, swing2LimitDegrees=1, twistLimitDegrees=1}
 {bone1Name=dyn_r_topLid, bone2Name=head, swing1LimitDegrees=60, swing2LimitDegrees=1, twistLimitDegrees=1}
 {bone1Name=dyn_r_btmLid, bone2Name=head, swing1LimitDegrees=60, swing2LimitDegrees=1, twistLimitDegrees=1}
]
