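# Physics asset for the wire part of the female medium Wicked Duck player cosmetic, covering a dynamic wire bone plus the head and neck.
# Four capsule bodies form a chain from neck_01 through neck_02 and head to dyn_wire_01_a, with the large head capsule enclosing the most volume.
# Neck and head joints leave swing and twist free, while the wire-to-head joint sets a much tighter twist limit than swing.
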PhysicsAsset /BRCosmetics/Characters/Player/Female/Medium/Bodies/F_MED_Wicked_Duck/Meshes/Parts/F_MED_Wicked_Duck_Wire_Physics
bodies {
 dyn_wire_01_a [{shapeType=Capsule, length=3.8645, radius=0.505, center=<2.302 -0.0016 -0.4532>, rotation=<rot -79.6559 179.9995 -179.9995>}]
 head [{shapeType=Capsule, length=2.4316, radius=11.767, center=<7.4641 0 0.0001>, rotation=<rot -90 0 0>}]
 neck_01 [{shapeType=Capsule, length=2, radius=2}]
 neck_02 [{shapeType=Capsule, length=2, radius=2}]
}
constraints [
 {
  bone1Name=head
  bone2Name=neck_02
  swing1LimitDegrees=45
  swing1Motion=Free
  swing2LimitDegrees=45
  swing2Motion=Free
  twistLimitDegrees=45
  twistMotion=Free
 }
 {
  bone1Name=neck_02
  bone2Name=neck_01
  swing1LimitDegrees=45
  swing1Motion=Free
  swing2LimitDegrees=45
  swing2Motion=Free
  twistLimitDegrees=45
  twistMotion=Free
 }
 {bone1Name=dyn_wire_01_a, bone2Name=head, swing1LimitDegrees=45, swing2LimitDegrees=45, twistLimitDegrees=5}
]
