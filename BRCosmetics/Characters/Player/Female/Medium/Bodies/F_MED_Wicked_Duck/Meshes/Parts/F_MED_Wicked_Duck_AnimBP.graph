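# Anim Blueprint for the Wicked Duck character part, animating headpiece, wire, eyes and eyelids over the parent mesh's pose.
# Cached base pose constrains headbob bones to the head, adds RigidBody physics (wire weaker when targeting) and drives lids from head-to-neck rotation.
# A DamageReceived stat increase sets the misleadingly named Active_Kill for 60 seconds.
# That flag blends from an offset headbob to an unshifted one with headpiece physics.
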AnimBlueprint /BRCosmetics/Characters/Player/Female/Medium/Bodies/F_MED_Wicked_Duck/Meshes/Parts/F_MED_Wicked_Duck_AnimBP
target_skeleton=&/F_MED_Wicked_Duck_Skeleton
graphs [
 AnimGraph {nodes [
   n1 = "Transform (Modify) Bone - Bone: dyn_headbob_MSTR"(ComponentPose=Local To Component(LocalPose=Use cached pose 'NoHit'()), Alpha=1, Translation="0,0,0", Rotation="0,0,0", Scale="1,1,1") {
    rotationMode=BMM_Additive
    rotationSpace=BCS_ParentBoneSpace
    scaleMode=BMM_Additive
    scaleSpace=BCS_ParentBoneSpace
    translationMode=BMM_Additive
    translationSpace=BCS_ParentBoneSpace
    alphaBoolBlend {blendOption=Linear}
    boneToModify {boneName=dyn_headbob_MSTR}
   }
   n2 = RigidBody(ComponentPose=@n1, Alpha=SelectFloat(A=0.2, B=0.4, bPickA=Get bIsTargeting())) {
    bClampLinearTranslationLimitToRefPose=true
    bOverrideWorldGravity=true
    overridePhysicsAsset=&/F_MED_Wicked_Duck_HeadPiecePhysics
    simulationSpace=BaseBoneSpace
    alphaBoolBlend {blendOption=Linear}
    baseBoneRef {boneName=root}
    componentAppliedLinearAccClamp=<1500 1500 1500>
    overrideWorldGravity {z=-980}
   }
   n3 = "Transform (Modify) Bone - Bone: dyn_headbob_MSTR"(ComponentPose=Local To Component(LocalPose=Use cached pose 'NoHit'()), Alpha=1, Translation="-7,0,0", Rotation="0,0,0", Scale="1,1,1") {
    rotationMode=BMM_Additive
    rotationSpace=BCS_ParentBoneSpace
    scaleMode=BMM_Additive
    scaleSpace=BCS_ParentBoneSpace
    translationMode=BMM_Additive
    translationSpace=BCS_ParentBoneSpace
    alphaBoolBlend {blendOption=Linear}
    boneToModify {boneName=dyn_headbob_MSTR}
    translation {x=-7}
   }
   Output Pose(Result=Blend Poses by bool(
     BlendPose_0=Component To Local(ComponentPose=@n2)
     BlendPose_1=Component To Local(ComponentPose=@n3)
     BlendTime_0=0.125
     BlendTime_1=0.05
     bActiveValue=Get Active_Kill() -> (Active_Kill:bool)
    ) {blendType=Linear}) {layerGroup=None}
   n4 = "Constraint - dyn_headbob_MSTR (P)"(
    ComponentPose=Local To Component(LocalPose=Copy Pose From Mesh() {bCopyCurves=true, bUseAttachedParent=true})
    Alpha=1
    ConstraintWeights_0=1
   ) {
    alphaBoolBlend {blendOption=Linear}
    boneToModify {boneName=dyn_headbob_MSTR}
    constraintSetup [{targetBone {boneName=head}}]
    constraintWeights [1]
   }
   n5 = "Constraint - dyn_headbob (P)"(ComponentPose=@n4, Alpha=1, ConstraintWeights_0=1) {
    alphaBoolBlend {blendOption=Linear}
    boneToModify {boneName=dyn_headbob}
    constraintSetup [{targetBone {boneName=head}}]
    constraintWeights [1]
   }
   n6 = RigidBody(ComponentPose=@n5, Alpha=SelectFloat(A=0.4, B=0.7, bPickA=Get bIsTargeting())) {
    bClampLinearTranslationLimitToRefPose=true
    bOverrideWorldGravity=true
    overridePhysicsAsset=&/F_MED_Wicked_Duck_Wire_Physics
    simulationSpace=BaseBoneSpace
    alphaBoolBlend {blendOption=Linear}
    baseBoneRef {boneName=root}
    componentAppliedLinearAccClamp=<1500 1500 1500>
    componentLinearAccScale=<0.4 0.4 0.5>
    overrideWorldGravity {z=-980}
   }
   n7 = RigidBody(ComponentPose=@n6, Alpha=0.3) {
    bClampLinearTranslationLimitToRefPose=true
    bOverrideWorldGravity=true
    overridePhysicsAsset=&/F_MED_Wicked_Duck_Eyes_Physics
    simulationSpace=BaseBoneSpace
    alphaBoolBlend {blendOption=Linear}
    baseBoneRef {boneName=root}
    componentAppliedLinearAccClamp=<1500 1500 1500>
    overrideWorldGravity {z=-980}
   }
   n8 = Get Head_LookUp() -> (Head_LookUp:double)
   n9 = "Transform (Modify) Bone - Bone: dyn_l_topLid"(ComponentPose=@n7, Alpha=@n8, Translation="0,0,0", Rotation="0,-10,0", Scale="1,1,1") {
    rotationMode=BMM_Additive
    rotationSpace=BCS_BoneSpace
    scaleMode=BMM_Additive
    scaleSpace=BCS_BoneSpace
    translationMode=BMM_Additive
    translationSpace=BCS_BoneSpace
    alphaBoolBlend {blendOption=Linear}
    boneToModify {boneName=dyn_l_topLid}
    rotation {yaw=-10}
   }
   n10 = "Transform (Modify) Bone - Bone: dyn_r_topLid"(ComponentPose=@n9, Alpha=@n8, Translation="0,0,0", Rotation="0,-10,0", Scale="1,1,1") {
    rotationMode=BMM_Additive
    rotationSpace=BCS_BoneSpace
    scaleMode=BMM_Additive
    scaleSpace=BCS_BoneSpace
    translationMode=BMM_Additive
    translationSpace=BCS_BoneSpace
    alphaBoolBlend {blendOption=Linear}
    boneToModify {boneName=dyn_r_topLid}
    rotation {yaw=-10}
   }
   n11 = Get Head_LookDown() -> (Head_LookDown:double)
   n12 = "Transform (Modify) Bone - Bone: dyn_l_topLid"(ComponentPose=@n10, Alpha=@n11, Translation="0,0,0", Rotation="0,10,0", Scale="1,1,1") {
    rotationMode=BMM_Additive
    rotationSpace=BCS_BoneSpace
    scaleMode=BMM_Additive
    scaleSpace=BCS_BoneSpace
    translationMode=BMM_Additive
    translationSpace=BCS_BoneSpace
    alphaBoolBlend {blendOption=Linear}
    boneToModify {boneName=dyn_l_topLid}
    rotation {yaw=10}
   }
   n13 = "Transform (Modify) Bone - Bone: dyn_r_topLid"(ComponentPose=@n12, Alpha=@n11, Translation="0,0,0", Rotation="0,10,0", Scale="1,1,1") {
    rotationMode=BMM_Additive
    rotationSpace=BCS_BoneSpace
    scaleMode=BMM_Additive
    scaleSpace=BCS_BoneSpace
    translationMode=BMM_Additive
    translationSpace=BCS_BoneSpace
    alphaBoolBlend {blendOption=Linear}
    boneToModify {boneName=dyn_r_topLid}
    rotation {yaw=10}
   }
   Save cached pose 'NoHit'(Pose=Component To Local(ComponentPose=@n13)) {cacheName=NoHit}
  ]}
 EventGraph {nodes [
   Event OnCharacterCustomizationLoaded()
   > n1 = Cast To FortPlayerPawn(Object=TryGetFortPawn())
   > n2 = WatchCosmeticStat!(
    execute:Exec
    Pawn:Fort Player Pawn Object Reference=@n1
    StatTag:Fort Stat Manager Tag Structure=(TagName="StatManager.DamageReceived")
    DefaultValue:int=0
    bBroadcastInitialValue:bool=true
    InitialValueBroadcastDelay:float=-1
    ChangeValueBroadcastDelay:float=-1
    bConsumeWaitingNotifications:bool=true
   ) -> (then:Exec, OnStatChanged:Exec, StatName:Name, StatValue:int) {is=title K2Node_AsyncAction}
   Set Did_Kill(execute=@n2.OnStatChanged, Did_Kill:bool=false) -> (Output_Get:bool)
   > n3 = Set KillCount(KillCount:double="To Float (Integer)"(InInt=@n2.StatValue)) -> (Output_Get:double)
   > Branch(Condition="float > float"(A=@n3, B=Get Last_KillCount() -> (Last_KillCount:double)) {is=title K2Node_CallFunction})
   > Set Active_Kill(Active_Kill:bool=true) -> (Output_Get:bool)
   > Delay(Duration=60)
   > Set Active_Kill(Active_Kill=false)
   > Set Last_KillCount(Last_KillCount:double=@n3) -> (Output_Get:double)
   Event BlueprintUpdateAnimation()
   n4 = BreakTransform(InTransform=GetDeltaTransformFromRefPose(self=GetOwnerSkeletalMeshComponent(), BoneName=head, BaseName=neck_02))
   > Set Head_LookUp(Head_LookUp:double=MapRangeClamped(Value=BreakRotator(InRot=@n4.Rotation).Yaw, InRangeA=-18, InRangeB=0, OutRangeA=1, OutRangeB=0)) -> (Output_Get:double)
   n5 = BreakTransform(InTransform=GetDeltaTransformFromRefPose(self=GetOwnerSkeletalMeshComponent(), BoneName=head, BaseName=neck_02))
   > Set Head_LookDown(Head_LookDown:double=MapRangeClamped(Value=BreakRotator(InRot=@n5.Rotation).Yaw, InRangeA=0, InRangeB=15)) -> (Output_Get:double)
  ]}
]
variables [
 Active Damaged {type=boolean}
 Active_Kill {type=boolean}
 Damaged Count {type=number}
 Damaged Recieved {type=boolean}
 Did_Kill {type=boolean}
 Head_LookDown {type=number}
 Head_LookUp {type=number}
 HeadBob_UpDown {type=Vector}
 KillCount {type=number}
 Last Damaged Recieved {type=number}
 Last_KillCount {type=number}
 Pawn {type=FortPlayerPawn}
 Skel_Mesh {type=SkeletalMeshComponent}
]
